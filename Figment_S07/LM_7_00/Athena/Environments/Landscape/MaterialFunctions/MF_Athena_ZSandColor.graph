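# Material function for the Athena landscape that outputs a sand tint varying with world height.
# World Z is shifted by an offset parameter, divided by a range parameter and clamped to form a gradient mask.
# That mask blends between low and high sand tints read from the MPC_Landscape_S07 parameter collection.
# A gradient texture sample and two same-named vector parameters are present but do not feed the output.
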MaterialFunction /Figment_S07/LM_7_00/Athena/Environments/Landscape/MaterialFunctions/MF_Athena_ZSandColor
graph {nodes [
  n1 = CollectionParameter() {collection=/Game/Athena/Environments/Landscape/MPC/MPC_Landscape_S07, parameterName=Sand_Low_Tint}
  n2 = CollectionParameter() {collection=&MPC_Landscape_S07, parameterName=Sand_High_Tint}
  n3 = ComponentMask(None=WorldPosition().XYZ) {b=true}
  n4 = ScalarParameter() {defaultValue=-1000, parameterName=Sand_Grad_Offset}
  n5 = Clamp(None=Divide(A=Add(A=@n3, B=@n4), B=ScalarParameter() {defaultValue=3500, parameterName=Sand_Grad_Range}))
  FunctionOutput(None=LinearInterpolate(A=ComponentMask(None=@n1) {b=true, g=true, r=true}, B=ComponentMask(None=@n2) {b=true, g=true, r=true}, Alpha=@n5))
  TextureSample(UVs=@n5) {texture=/Figment_S07/LM_7_00/Athena/Environments/Landscape/Textures/Arid/T_Athena_Arid_Gradient_02}
  VectorParameter() {parameterName=Sand_Low_Tint, defaultValue=<rgba 1 0.6211 0.5813 0>}
  VectorParameter() {parameterName=Sand_High_Tint, defaultValue=<rgba 0.9115 0.7687 0.7778 0>}
 ]}
