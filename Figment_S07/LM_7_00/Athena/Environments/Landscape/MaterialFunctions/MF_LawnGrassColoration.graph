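# Landscape material function that recolors grass for golf fairways and greens, with Standard and Low material variants plus tint and green-mask outputs.
# Base color red remaps into a tinted grass color, and rotated landscape coordinates, fracted and rounded, add fairway mowing stripes.
# Standard height-blends fairway edges using inverted opacity, then brightens and normal-flattens the painted Golf_Green layer.
# Low uses the plain fairway mask and drops stripes on ES2/ES3.1.
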MaterialFunction /Figment_S07/LM_7_00/Athena/Environments/Landscape/MaterialFunctions/MF_LawnGrassColoration
graph {nodes [
  n1 = FunctionInput() {bUsePreviewValueAsDefault=true, inputName=Material_Low, inputType=FunctionInput_MaterialAttributes, sortPriority=1}
  n2 = ScalarParameter() {defaultValue=0.075, parameterName=Golf_Grass_Min}
  n3 = ScalarParameter() {defaultValue=1, parameterName=Golf_Grass_Max}
  n4 = MaterialFunctionCall!(Material=@n1) {materialFunction=/Engine/Functions/MaterialLayerFunctions/MatLayerBlend_BreakBaseColor}
  n5 = VectorParameter() {parameterName=Golf_GrassColor, defaultValue=<rgba 0.5 1.05 0.2516 1>}
  n6 = Multiply(A=LinearInterpolate(A=@n2, B=@n3, Alpha=ComponentMask(None=@n4.BaseColor) {r=true}), B=@n5.RGB)
  n7 = FunctionInput() {bUsePreviewValueAsDefault=true, inputName=Golf_Fairway_Mask, inputType=FunctionInput_Scalar, sortPriority=2}
  n8 = ScalarParameter() {defaultValue=1.1, parameterName=Fairway_Color_Max}
  n9 = LandscapeLayerCoords() {mappingScale=4}
  n10 = ScalarParameter() {defaultValue=0.4, parameterName=gfg}
  n11 = Frac(None=ComponentMask(None=MaterialFunctionCall(UVs=@n9, "Rotation Angle (0-1)"=@n10) {materialFunction=&CustomRotator}.Rotated Values) {r=true})
  n12 = LinearInterpolate(A=ScalarParameter() {defaultValue=0.9, parameterName=Fairway_Color_Min}, B=@n8, Alpha=Clamp(None=Round(None=@n11)))
  n13 = FeatureLevelSwitch(Default=@n12, ES2=@n8, ES3_1=@n8, SM4_REMOVED=@n12, SM5=@n12, SM6=@n12)
  n14 = Clamp(None=LinearInterpolate(A=ScalarParameter() {defaultValue=-4, parameterName=dfsdfsd}, B=ScalarParameter() {defaultValue=1, parameterName=sdghfghf}, Alpha=@n7))
  FunctionOutput(None=MaterialFunctionCall!(
    Material=MaterialFunctionCall(Material=@n1, NewBaseColor=@n6, Mask=@n7) {materialFunction=/Engine/Functions/MaterialLayerFunctions/MatLayerBlend_OverrideBaseColor}.Blended Material
    NewBaseColor=Multiply(A=@n6, B=@n13)
    Mask=@n14
   ) {materialFunction=&MatLayerBlend_OverrideBaseColor}.Blended Material) {outputName=Material_Low, sortPriority=1}
  n15 = FunctionInput() {bUsePreviewValueAsDefault=true, inputType=FunctionInput_MaterialAttributes}
  n16 = MaterialFunctionCall(Material=@n15) {materialFunction=&MatLayerBlend_BreakBaseColor}
  n17 = Multiply(A=LinearInterpolate(A=@n2, B=@n3, Alpha=ComponentMask(None=@n16.BaseColor) {r=true}), B=@n5.RGB)
  n18 = MaterialFunctionCall(
   Transition Phase=@n7
   Height Texture=OneMinus(None=MaterialFunctionCall!(Mat attributes=@n15) {materialFunction=/Engine/Functions/Engine_MaterialFunctions02/MatLayerBlend_BreakOpacity}.Result)
   Contrast=ScalarParameter() {defaultValue=1, parameterName=Golf_EdgeContrast}
  ) {materialFunction=&HeightLerp}
  n19 = MaterialFunctionCall(
   Material=MaterialFunctionCall(Material=@n15, NewBaseColor=@n17, Mask=@n18.Alpha) {materialFunction=&MatLayerBlend_OverrideBaseColor}.Blended Material
   NewBaseColor=Multiply(A=@n17, B=@n12)
   Mask=@n14
  ) {materialFunction=&MatLayerBlend_OverrideBaseColor}
  n20 = Multiply(A=@n17, B=ScalarParameter() {defaultValue=1.25, parameterName=Green_Color_Max})
  n21 = LandscapeLayerSample() {parameterName=Golf_Green}
  FunctionOutput(None=MaterialFunctionCall!(
    Material=MaterialFunctionCall(Material=@n19.Blended Material, NewBaseColor=@n20, Mask=@n21) {materialFunction=&MatLayerBlend_OverrideBaseColor}.Blended Material
    NormalFlatness=@n21
    Normal
   ) {materialFunction=/Engine/Functions/MaterialLayerFunctions/MatLayerBlend_NormalFlatten}.Blended Material) {outputName=Material_Standard}
  FunctionOutput(None=Multiply(A=@n12, B=ScalarParameter() {defaultValue=1.25, parameterName=Golf_LowEnd_FairwayTint})) {outputName=Fairway_Color, sortPriority=2}
  FunctionOutput(None=ScalarParameter() {defaultValue=1.375, parameterName=Golf_LowEnd_GreenTint}) {outputName=Golf_Green_Color, sortPriority=3}
  FunctionOutput(None=@n21) {outputName=Green_Mask, sortPriority=4}
  MaterialFunctionCall() {materialFunction=&MatLayerBlend_BreakOpacity}
  Multiply(A=Power(Base=@n18.Alpha, Exp=ScalarParameter() {defaultValue=0.5, parameterName=EdgeShadowTest}), B=@n17)
  QualitySwitch()
  ScalarParameter() {defaultValue=1.3, parameterName=Golf_EdgeBrightness}
 ]}
