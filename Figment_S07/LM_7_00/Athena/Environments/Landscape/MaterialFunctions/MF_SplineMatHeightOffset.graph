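# Material function that computes a camera-facing offset for landscape spline materials, apparently for World Position Offset, to counter distance Z-fighting.
# It normalizes the vector from world position to camera and scales it by a blended bias amount.
# A camera-centered SphereMask, tuned by start radius and hardness parameters, blends from a one-unit offset nearby to the larger DistanceZFightBias parameter far away.
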MaterialFunction /Figment_S07/LM_7_00/Athena/Environments/Landscape/MaterialFunctions/MF_SplineMatHeightOffset
graph {nodes [
  n1 = CameraPositionWS()
  n2 = WorldPosition()
  n3 = Subtract(A=@n1, B=@n2.XYZ)
  n4 = ScalarParameter() {defaultValue=200, group=SplineHeightOffset, parameterName=DistanceZFightBias}
  n5 = SphereMask(
   A=@n1
   B=@n2.XYZ
   Radius=ScalarParameter() {defaultValue=40000, parameterName=ZBiasStart}
   Hardness=ScalarParameter() {defaultValue=1, parameterName=ZBiasHardness}
  )
  FunctionOutput(None=Multiply(A=Normalize(VectorInput=@n3), B=LinearInterpolate(A=@n4, Alpha=@n5)))
 ]}
